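# Niagara effect for the SizeTribe companion emote, showing seven spinning golden toon rings with flashes and glitter.
# Ring emitters loop seven times, placing each burst on the socket or bone matching the loop index, with staggered lifetimes retiring rings in reverse order.
# GlitterSpawn001 adds sparkle bursts there, while GlitterFalling001 samples Rings_Toon_High particles to shed falling glitter.
# Spawn-group switches merge several looks per emitter, and time-of-day inputs tune color.
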NiagaraSystem /CosmeticCompanions/Assets/Biped/SizeTribe/FX/Emote/NS_CompanionEmote_SizeTribe_Rings
dependencies {
 usedDataInterfaces [
  &NiagaraDataInterfaceCurve
  &NiagaraDataInterfaceParticleRead
  &NiagaraDataInterfaceVectorCurve
  &NiagaraDataInterfaceVector2DCurve
  &NiagaraDataInterfaceColorCurve
  &NiagaraDataInterfaceArrayFloat
  &NiagaraDataInterfaceSkeletalMesh
 ]
 usedDynamicInputs [
  /Niagara/DynamicInputs/Multiply/ScaleAndBiasFloat
  &ColorFromCurve
  &FNTimeOfDayColor
  &SelectFloatFromArray
 ]
 usedModules [
  &self:SystemSpawnScript.SystemScriptSource.SystemScriptGraph.NiagaraNodeAssignment_1.SetVariables_#1
  &SystemState
  &self:SystemSpawnScript.SystemScriptSource.SystemScriptGraph.NiagaraNodeAssignment_0.SetVariables_#2
 ]
}
emitters {
 GlitterFalling001 {
  emitterUpdateScript [
   &EmitterState(
    Enable Distance Culling=true
    Life Cycle Mode=::NewEnumerator1
    Loop Behavior=::NewEnumerator1
    Loop Delay=1.5
    Loop Duration=6
    MaxDistance=10000
    Module.bApplySpawnCountScaleByDistance=true
    Module.bUseMaxDistance=true
    Module.bUseMinDistance=false
    Scalability Mode=::NewEnumerator1
    Scale Spawn Count=true
    Spawn Count Scale By Distance=DataInterface() {propertyValues {Curve {keys [<at 0 1>,<at 0.5 0.25>]}}}
    UseLoopDelay=false
   )
   &SpawnRate(SpawnRate=1500, Use Spawn Probability=false) {enabled=false}
   &SpawnParticlesFromOtherEmitter(
    Attribute Reader=DataInterface() {propertyValues {EmitterBinding {bindingMode=Other, emitterName=Rings_Toon_High}}}
    Spawn Rate=35
    Use Maximum Age=false
    Use Minimum Age=false
   )
  ]
  particleSpawnScript [
   &InitializeParticle(
    Color Maximum=<rgba 1 0.45 0 1>
    Color Minimum=<rgba 1 0.5281 0.0563 1>
    Color Mode=::NewEnumerator2
    Lifetime Max=0.8
    Lifetime Min=0.3
    Lifetime Mode=::NewEnumerator1
    Sprite Size Mode=::NewEnumerator3
    Uniform Sprite Size=&Lerp_Float(
     A=5
     B=50
     Alpha=&FloatFromCurve(
      FloatCurve=DataInterface() {propertyValues {Curve {keys [
          {interpMode=RCIM_Cubic, tangentMode=RCTM_User, time=0, value=0.2}
          {arriveTangent=1.1314, interpMode=RCIM_Cubic, leaveTangent=1.1314, tangentMode=RCTM_User, time=0.9, value=0.6}
          {interpMode=RCIM_Cubic, tangentMode=RCTM_User, time=1, value=1}
         ]}}}
      CurveIndex=&RandomRangeFloat(Minimum=0, Maximum=1)
     )
    )
    UsePositionOffset=false
    Write Material Random=true
   ) {module=InitializeParticle001}
   &SubUVAnimation(End Frame=3, SubUV Animation Mode:ENiagara_SubUVLookupMode=::NewEnumerator2)
   &SampleParticlesFromOtherEmitter(
    Applied Position Offset:Vector3f=<0 0 15>
    Maximum Sampled Particle Age=1.75
    Minimum Sampled Particle Age:NiagaraFloat=&FloatFromCurve(
     FloatCurve=DataInterface() {propertyValues {Curve {keys [<at 0 1 RCIM_Cubic>, {arriveTangent=-1.4825, interpMode=RCIM_Cubic, leaveTangent=-1.4825, tangentMode=RCTM_User, time=2, value=0}]}}}
     CurveIndex=$Emitter.Age
     Scale Curve=0.9
    )
    Module.MaxSampledAgeCutoff=false
    Module.MinSampledAgeCutoff=true
    Reject Particle By Sampled Age=true
    Reject Particle on Sampled Particle's First Frame=true
   )
   &ShapeLocation(
    Additional Quaternion Rotation=true
    Additional Yaw / Pitch / Roll=&MakeVector(X=0, Y=0, Z=&Multiply_Float(A=200, B=$Engine.System.Age))
    Handle Radius=0.5
    Large Radius=11
    MaskSpawnGroup=false
    Offset Mode=::NewEnumerator0
    Override Seed=false
    Rotation Angle Type:ENiagara_AngleInput=::NewEnumerator0
    Rotation Mode=::NewEnumerator3
    Rotation Quaternion=&MakeQuaternion(
     Quaternion From=0
     Axis=<0 1 0>
     Angle=90
     From=?
     To=?
     Stabilization Mode:ENiagara_OrientVectorCount=::NewEnumerator0
     Side Direction:Vector3f=?
     Up Direction:Vector3f=?
     Angle Type=::NewEnumerator0
     Normalize:NiagaraBool=true
    )
    Shape Primitive=::NewEnumerator3
    Surface Distribution=1
    UseOwnerScale=true
   )
  ]
  particleUpdateScript [
   &ParticleState()
   &ScaleSpriteSize(
    Uniform Curve Scale=&FN_LOD_Distance_Remap(Start Distance=500, End Distance=2500, Minimum Output=0.25, Maximum Output=1, Clamp Output=true)
    Uniform Curve Sprite Scale=DataInterface() {propertyValues {Curve {keys [
        {interpMode=RCIM_Cubic, tangentMode=RCTM_User, time=0, value=0}
        {interpMode=RCIM_Cubic, tangentMode=RCTM_User, time=0.2, value=1}
        {tangentMode=RCTM_User, time=1, value=0}
       ]}}}
   )
   &SpriteRotationRate()
   &GravityForce(Gravity=&RandomRangeVector(Minimum=<0 0 -50>, Maximum=<0 0 -200>))
   &CurlNoiseForce(Bias Noise=false, Pan Noise=false)
   &SolveForcesAndVelocity(Clamp Velocity=false, Limit Acceleration=false)
   &ScaleColor(
    Scale Alpha=&FloatFromCurve(
     FloatCurve=DataInterface() {propertyValues {Curve {keys [<at 0 1>,<at 1 0>]}}}
     Scale Curve=&Waveform(
      Global Amplitude Scale=10
      "Waveform [1]"=::NewEnumerator0
      "[1] Phase Scale"=&RandomRangeFloat(Minimum=2, Maximum=4)
      "[1] Frequency"=&RandomRangeFloat(Minimum=0.75, Maximum=1)
      "[1] Amplitude Min/Max"=<0 1>
     )
    )
    ScaleA=true
    ScaleRGB=true
   )
   &DynamicMaterialParameters(
    Index 0 Param 1=0.15
    Index 0 Param 2=&RandomRangeFloat(Minimum=0.25, Maximum=2)
    Index 0 Param 3=0.5
    Module.Index0Param0Write=true
    Module.Index0Param1Write=true
    Module.Index0Param2Write=true
    Module.Index0Param3Write=true
   )
   /SetVariables(
    Particles.Linear Color=&Multiply_LinearColorByFloat(Linear Color=$Particles.Color, Float=2)
    Particles.Vector 2D:Vector2f=&Multiply_Vector2D_ByFloat(Float=&RandomRangeFloat(Minimum=0.15, Maximum=0.25), Vector2D=$Particles.SpriteSize)
   ) {isSetParameters=true}
  ]
  rendererClasses [&NiagaraSpriteRendererProperties]
  renderers [
   {rendererClass=&NiagaraSpriteRendererProperties}
   {rendererClass=&NiagaraSpriteRendererProperties, rendererIndex=1}
  ]
 }
 GlitterSpawn001 {
  emitterUpdateScript [
   &EmitterState(Life Cycle Mode=::NewEnumerator1, Loop Behavior=::NewEnumerator2, Loop Count=7, Loop Duration=0.25, UseLoopDelay=false)
   &SpawnBurst_Instantaneous(
    Spawn Count=&RandomRangeInt(Minimum=20, Maximum=35)
    Use Spawn Probability=false
    UseLoopCountLimit=false
   )
  ]
  particleSpawnScript [
   &InitializeParticle(
    Color Maximum=<rgba 1 0.45 0 1>
    Color Minimum=<rgba 1 0.5281 0.0563 1>
    Color Mode=::NewEnumerator2
    Lifetime Max=0.8
    Lifetime Min=0.3
    Lifetime Mode=::NewEnumerator1
    Sprite Size Mode=::NewEnumerator3
    Uniform Sprite Size=&Lerp_Float(
     A=5
     B=50
     Alpha=&FloatFromCurve(
      FloatCurve=DataInterface() {propertyValues {Curve {keys [
          {interpMode=RCIM_Cubic, tangentMode=RCTM_User, time=0, value=0.2}
          {arriveTangent=1.1314, interpMode=RCIM_Cubic, leaveTangent=1.1314, tangentMode=RCTM_User, time=0.9, value=0.6}
          {interpMode=RCIM_Cubic, tangentMode=RCTM_User, time=1, value=1}
         ]}}}
      CurveIndex=&RandomRangeFloat(Minimum=0, Maximum=1)
     )
    )
    UsePositionOffset=false
    Write Material Random=true
   )
   /SetVariables(Particles.PartitionParticles.Partition:NiagaraInt32=$Emitter.LoopCount) {isSetParameters=true}
   &SkeletalMeshLocation(
    Bone / Socket Index=$Particles.PartitionParticles.Partition
    Bone Sampling Mode=::NewEnumerator4
    MaskSpawnGroup=false
    Orientation Sampling=::NewEnumerator1
    Sampled Position Offset Coordinate Space:ENiagaraSkelMeshTransforms=::NewEnumerator0
    Skeletal Mesh=$System.Skeletal Mesh Reader
   ) {module=SkeletalMeshLocation001}
   &ShapeLocation(
    MaskSpawnGroup=false
    Offset Mode=::NewEnumerator0
    Override Seed=false
    Rotation Axis=<0 1 0>
    Rotation Mode=::NewEnumerator1
    Shape Origin=$Particles.SkeletalMeshLocation001.SampledPosition
    Sphere Radius=&RandomRangeFloat(Minimum=5, Maximum=10)
    UseOwnerScale=true
   )
   &SubUVAnimation(End Frame=3, SubUV Animation Mode:ENiagara_SubUVLookupMode=::NewEnumerator2)
   &AddVelocity(
    Module.Use Falloff Distance=false
    Override Seed=false
    Velocity Mode=::NewEnumerator1
    Velocity Origin=$Particles.SkeletalMeshLocation001.SampledPosition
    Velocity Speed=&RandomRangeFloat(Minimum=50, Maximum=150)
   )
  ]
  particleUpdateScript [
   &ParticleState()
   &SpriteRotationRate()
   &GravityForce(Gravity=&RandomRangeVector(Minimum=<0 0 20>, Maximum=<0 0 -60>))
   &ScaleColor(
    Scale Alpha=&FloatFromCurve(
     FloatCurve=DataInterface() {propertyValues {Curve {keys [<at 0 1>,<at 1 0>]}}}
     Scale Curve=&Waveform(
      Global Amplitude Scale=10
      "Waveform [1]"=::NewEnumerator0
      "[1] Phase Scale"=&RandomRangeFloat(Minimum=2, Maximum=4)
      "[1] Frequency"=&RandomRangeFloat(Minimum=0.75, Maximum=1)
      "[1] Amplitude Min/Max"=<0 1>
     )
    )
    Scale RGB=<5 5 5>
    ScaleA=true
    ScaleRGB=true
   )
   &DynamicMaterialParameters(
    Index 0 Param 1=0.15
    Index 0 Param 2=&RandomRangeFloat(Minimum=0.25, Maximum=2)
    Index 0 Param 3=0.5
    Module.Index0Param0Write=true
    Module.Index0Param1Write=true
    Module.Index0Param2Write=true
    Module.Index0Param3Write=true
   )
   /SetVariables(
    Particles.Linear Color=&Multiply_LinearColorByFloat(Linear Color=$Particles.Color, Float=2)
    Particles.Vector 2D:Vector2f=&Multiply_Vector2D_ByFloat(Float=&RandomRangeFloat(Minimum=0.15, Maximum=0.25), Vector2D=$Particles.SpriteSize)
   ) {isSetParameters=true}
   &ScaleSpriteSize(
    Uniform Curve Scale=&FN_LOD_Distance_Remap(Start Distance=500, End Distance=2500, Minimum Output=0.25, Maximum Output=1, Clamp Output=true)
    Uniform Curve Sprite Scale=DataInterface() {propertyValues {Curve {keys [
        {interpMode=RCIM_Cubic, tangentMode=RCTM_User, time=0, value=0.2}
        {interpMode=RCIM_Cubic, tangentMode=RCTM_User, time=0.2, value=1}
        {tangentMode=RCTM_User, time=1, value=0}
       ]}}}
   )
   &SolveForcesAndVelocity(Clamp Velocity=false, Limit Acceleration=false)
  ]
  rendererClasses [&NiagaraSpriteRendererProperties]
  renderers [
   {rendererClass=&NiagaraSpriteRendererProperties}
   {rendererClass=&NiagaraSpriteRendererProperties, rendererIndex=1}
  ]
 }
 Rings_Toon2 {
  emitterUpdateScript [
   &EmitterState(
    Life Cycle Mode=::NewEnumerator1
    Loop Behavior=::NewEnumerator2
    Loop Count=7
    Loop Duration=0.25
    Recalculate Duration Each Loop=true
    UseLoopDelay=false
   )
   &SpawnBurst_Instantaneous(Loop Count Limit=5, Use Spawn Probability=false, UseLoopCountLimit=false)
   &SpawnBurst_Instantaneous(Use Spawn Probability=false, UseLoopCountLimit=false) {module=SpawnBurst_Instantaneous001}
   &SpawnBurst_Instantaneous(Spawn Group=1, Use Spawn Probability=false, UseLoopCountLimit=false) {module=SpawnBurst_Instantaneous002}
   &SpawnBurst_Instantaneous(
    Spawn Count=&RandomRangeInt(Minimum=20, Maximum=35)
    Spawn Group=2
    Use Spawn Probability=false
    UseLoopCountLimit=false
   ) {module=SpawnBurst_Instantaneous003}
   /SetVariables(Emitter.Emissive_Intensity:NiagaraFloat=&FNTimeOfDayScalar(Night Scale=0.4, Morning Scale=0.225, Day Scale=0.15, Evening Scale=0.25)) {isSetParameters=true}
  ]
  particleSpawnScript [
   /SetVariables(Particles.PartitionParticles.Partition:NiagaraInt32=$Emitter.LoopCount, Particles.VisibilityTag=&SpawnGroup()) {isSetParameters=true, module=SetVariables_#3}
   &InitializeParticle(
    Color=&MergeHelper_ColorSwitch(
     0=<rgba 2.3125 1.1562 0 1>
     1=<rgba 3 1.8435 0.2244 1>
     2=&RandomRangeLinearColor(Minimum=<rgba 1 0.5281 0.0563 1>, Maximum=<rgba 1 0.45 0 1>)
     3=&RandomRangeLinearColor(Minimum=<rgba 1 0.5281 0.0563 1>, Maximum=<rgba 1 0.45 0 1>)
     4=/Niagara/DynamicInputs/TypeConversions/MakeLinearColorFromFloat(Value:NiagaraFloat=&Lerp_Float(
       A=5
       B=50
       Alpha=&FloatFromCurve(
        FloatCurve=DataInterface() {propertyValues {Curve {keys [
            {interpMode=RCIM_Cubic, tangentMode=RCTM_User, time=0, value=0.2}
            {arriveTangent=1.1314, interpMode=RCIM_Cubic, leaveTangent=1.1314, tangentMode=RCTM_User, time=0.9, value=0.6}
            {interpMode=RCIM_Cubic, tangentMode=RCTM_User, time=1, value=1}
           ]}}}
        CurveIndex=&RandomRangeFloat(Minimum=0, Maximum=1)
       )
      ))
     Index=$Particles.VisibilityTag
     Num=::NewEnumerator1
    )
    Lifetime=&MergeHelper_FloatSwitch(
     0=&MergeHelper_FloatSwitch(0=4.948, 1=4.552, 2=4.034, 3=3.517, 4=3, 5=2.627, 6=2.167, Index=$Particles.PartitionParticles.Partition, Num=::NewEnumerator5)
     1=0.1
     2=&RandomRangeFloat(Minimum=0.3, Maximum=0.8)
     3=0.1
     4=&RandomRangeFloat(Minimum=0.3, Maximum=0.6)
     Index=$Particles.VisibilityTag
     Num=::NewEnumerator2
    )
    Mesh Scale Mode=::NewEnumerator3
    Mesh Uniform Scale=1.2
    Sprite Rotation Angle=&RandomRangeFloat(Minimum=-45, Maximum=45)
    Sprite Rotation Mode=::NewEnumerator0
    Sprite Size Mode=::NewEnumerator3
    Sprite UV Mode=::NewEnumerator1
    Uniform Sprite Size=&MergeHelper_FloatSwitch(
     0=0
     1=&MakeFloatFromVector2d(Vector 2D=&Multiply_Vector2D_ByFloat(Float=&RandomRangeFloat(Minimum=0.4, Maximum=0.525), Vector2D=<35 50>))
     2=&Lerp_Float(
      A=5
      B=50
      Alpha=&FloatFromCurve(
       FloatCurve=DataInterface() {propertyValues {Curve {keys [
           {interpMode=RCIM_Cubic, tangentMode=RCTM_User, time=0, value=0.2}
           {arriveTangent=1.1314, interpMode=RCIM_Cubic, leaveTangent=1.1314, tangentMode=RCTM_User, time=0.9, value=0.6}
           {interpMode=RCIM_Cubic, tangentMode=RCTM_User, time=1, value=1}
          ]}}}
       CurveIndex=&RandomRangeFloat(Minimum=0, Maximum=1)
      )
     )
     3=&MakeFloatFromVector2d(Vector 2D=&Multiply_Vector2D_ByFloat(Float=&RandomRangeFloat(Minimum=0.45, Maximum=0.7), Vector2D=<35 50>))
     Index=$Particles.VisibilityTag
     Num=::NewEnumerator1
    )
    UsePositionOffset=false
    Write Material Random=true
   )
   &InitialMeshOrientation(Mesh Orientation Mode=::NewEnumerator6, Orientation Quaternion:Quat4f=<1 0 0 1>, Use Rotation Vector=false)
   &SkeletalMeshLocation(
    Bone / Socket Index=&MergeHelper_IntSwitch(0=$Particles.PartitionParticles.Partition, 1=6, 2=6, 3=?, 4=?, 5=?, Index=$Particles.VisibilityTag, Num=::NewEnumerator1)
    Bone Sampling Mode=::NewEnumerator4
    MaskSpawnGroup=false
    Orientation Sampling=::NewEnumerator1
    Sampled Position Offset=&MergeHelper_VectorSwitch(0=<0 0 0>, 1=<0 0 2>, 2=<0 0 0>, 3=?, Index=$Particles.VisibilityTag, Num=::NewEnumerator1)
    Sampled Position Offset Coordinate Space:ENiagaraSkelMeshTransforms=::NewEnumerator0
    Skeletal Mesh=$System.Skeletal Mesh Reader
   ) {module=SkeletalMeshLocation001}
   /SetVariables(Particles.GlitterSubUV:NiagaraFloat=&RandomRangeFloat(Minimum=0, Maximum=3)) {isSetParameters=true, module=SetVariables_#4}
   /SetVariables(Particles.SpriteRotationStar:NiagaraFloat=&RandomRangeFloat(Minimum=-360, Maximum=360)) {isSetParameters=true}
   &ShapeLocation(
    MaskSpawnGroup=false
    Offset Mode=::NewEnumerator0
    Override Seed=false
    Rotation Axis=<0 1 0>
    Rotation Mode=::NewEnumerator1
    Shape Origin=&MergeHelper_VectorSwitch(
     0=Linked() {linkedVariable {name=StackContext.Position, type=NiagaraPosition}}
     1=Linked() {linkedVariable {name=StackContext.Position, type=NiagaraPosition}}
     2=Linked() {linkedVariable {name=Particles.SkeletalMeshLocation001.SampledPosition, type=NiagaraPosition}}
     3=?
     Index=$Particles.VisibilityTag
     Num=::NewEnumerator1
    )
    Sphere Radius=&MergeHelper_FloatSwitch(
     0=0
     1=0
     2=&RandomRangeFloat(Minimum=5, Maximum=10)
     3=0
     4=0
     Index=$Particles.VisibilityTag
     Num=::NewEnumerator3
    )
    UseOwnerScale=true
   )
   &AddVelocity(
    Module.Use Falloff Distance=false
    Origin Offset:NiagaraPosition=$Particles.SkeletalMeshLocation001.SampledPosition
    Override Seed=false
    Velocity Mode=::NewEnumerator1
    Velocity Speed=&MergeHelper_FloatSwitch(
     0=0
     1=0
     2=&RandomRangeFloat(Minimum=50, Maximum=150)
     3=?
     Index=$Particles.VisibilityTag
     Num=::NewEnumerator1
    )
   )
  ]
  particleUpdateScript [
   &ParticleState()
   &GravityForce(Gravity=&MergeHelper_VectorSwitch(
     0=<0 0 0>
     1=<0 0 0>
     2=&RandomRangeVector(Minimum=<0 0 20>, Maximum=<0 0 -60>)
     3=?
     Index=$Particles.VisibilityTag
     Num=::NewEnumerator1
    ))
   &ScaleMeshSize(Scale Factor=&VectorFromCurve(VectorCurve=DataInterface() {propertyValues {
       XCurve {keys [{arriveTangent=4.1361, interpMode=RCIM_Cubic, leaveTangent=4.1361, tangentMode=RCTM_User, time=0, value=0}, <at 0.3 1 RCIM_Cubic>]}
       YCurve {keys [<at 0 1>,<at 1 1>]}
       ZCurve {keys [{arriveTangent=4.1361, interpMode=RCIM_Cubic, leaveTangent=4.1361, tangentMode=RCTM_User, time=0, value=0}, <at 0.3 1 RCIM_Cubic>]}
      }}))
   &ScaleColor(
    Scale Alpha=&MergeHelper_FloatSwitch(
     0=1
     1=1
     2=&FloatFromCurve(
      FloatCurve=DataInterface() {propertyValues {Curve {keys [<at 0 1>,<at 1 0>]}}}
      Scale Curve=&Waveform(
       Global Amplitude Scale=10
       "Waveform [1]"=::NewEnumerator0
       "[1] Phase Scale"=&RandomRangeFloat(Minimum=2, Maximum=4)
       "[1] Frequency"=&RandomRangeFloat(Minimum=0.75, Maximum=1)
       "[1] Amplitude Min/Max"=<0 1>
      )
     )
     3=?
     Index=$Particles.VisibilityTag
     Num=::NewEnumerator1
    )
    Scale RGB=&MergeHelper_VectorSwitch(
     0=&FN_FrontEndVector3Modifier(
      Gameplay=&FN_Select_Vector_from_Time_Of_Day(Morning=<1.115 1.15 1.15>, Day=<1 1 1>, Evening=<1 1 1>, Night=<3.5 2.4 1>)
      Front End=<3.2 2.75 2.75>
     )
     1=<1 1 1>
     2=<5 5 5>
     3=?
     Index=$Particles.VisibilityTag
     Num=::NewEnumerator1
    )
    ScaleA=true
    ScaleRGB=true
   )
   &Drag(
    Drag=&MergeHelper_FloatSwitch(0=1, 1=1, 2=0, 3=?, Index=$Particles.VisibilityTag, Num=::NewEnumerator1)
    Use Linear Drag=true
    Use Rotational Drag=false
   )
   &SolveForcesAndVelocity(Clamp Velocity=false, Limit Acceleration=false)
   /SetVariables(
    Particles.DynamicMaterialParameterGlitter:Vector4f=&MakeVector4(X=0.15, Y=&RandomRangeFloat(Minimum=0.25, Maximum=2), Z=0.5, W=0)
    Particles.Linear Color=&Multiply_LinearColorByFloat(Linear Color=$Particles.Color, Float=2)
    Particles.SmallGlitterSize:Vector2f=&Multiply_Vector2D_ByFloat(Float=&RandomRangeFloat(Minimum=0.15, Maximum=0.25), Vector2D=$Particles.SpriteSize)
   ) {isSetParameters=true}
   &ScaleSpriteSize(
    Scale Factor=&MergeHelper_Vector2Switch(
     0=<0 0>
     1=&Vector2DFromCurve(
      Vector2Curve=DataInterface() {propertyValues {
        XCurve {
         defaultValue=0
         keys [
          {arriveTangent=0.0865, interpMode=RCIM_Cubic, leaveTangent=0.0865, tangentMode=RCTM_User, time=0, value=0}
          {arriveTangent=4.9825, interpMode=RCIM_Cubic, leaveTangent=4.9825, tangentMode=RCTM_User, time=0.4724, value=1}
         ]
        }
        YCurve {
         defaultValue=1
         keys [
          {arriveTangent=-5.5781, interpMode=RCIM_Cubic, leaveTangent=-5.5781, tangentMode=RCTM_User, time=0.6, value=1.1}
          {arriveTangent=-0.1643, interpMode=RCIM_Cubic, leaveTangent=-0.1643, tangentMode=RCTM_User, time=1, value=0}
         ]
        }
       }}
      ScaleCurve=&RandomRangeVector2D(Minimum=<1 1>, Maximum=<1 1.2>)
     )
     2=&Vector2DFromFloat(Value=&FloatFromCurve(
       FloatCurve=DataInterface() {propertyValues {Curve {keys [
           {interpMode=RCIM_Cubic, tangentMode=RCTM_User, time=0, value=0.2}
           {interpMode=RCIM_Cubic, tangentMode=RCTM_User, time=0.2, value=1}
           {tangentMode=RCTM_User, time=1, value=0}
          ]}}}
       Scale Curve=&FN_LOD_Distance_Remap(Start Distance=500, End Distance=2500, Minimum Output=0.25, Maximum Output=1, Clamp Output=true)
      ))
     3=?
     4=?
     Index=$Particles.VisibilityTag
     Num=::NewEnumerator1
    )
    Scale Sprite Size Mode=::NewEnumerator2
   ) {module=ScaleSpriteSize001}
   /SetVariables(
    Particles.CameraOffsetStar=15
    Particles.GlowColor=&MakeLinearColorFromVectorAndFloat(
     "Vector (RGB)"=&Multiply_Vector(
      A=&MakeVectorFromLinearColorRGB(Color=<rgba 1.2 0.6795 0 1>)
      B=&MakeVectorFromLinearColorRGB(Color=$System.TOD)
     )
     "Float (Alpha)"=1
    )
    Particles.GlowSize=&Multiply_Vector2D_ByFloat(Float=1.75, Vector2D=$Particles.SpriteSize)
    Particles.StarPosition:NiagaraPosition=$Particles.Initial.Position
   ) {isSetParameters=true, module=SetVariables_#5}
   &SkeletalMeshLocation(
    Bone / Socket Index=$Particles.PartitionParticles.Partition
    Bone Sampling Mode=::NewEnumerator4
    MaskSpawnGroup=false
    Orientation Sampling=::NewEnumerator1
    Position Sampling=::NewEnumerator2
    Sampled Position Offset=&MergeHelper_VectorSwitch(0=<0 0 0>, 1=<0 0 3>, 2=<0 0 3>, 3=?, Index=$Particles.VisibilityTag, Num=::NewEnumerator1)
    Sampled Position Offset Coordinate Space:ENiagaraSkelMeshTransforms=::NewEnumerator0
    Skeletal Mesh=$System.Skeletal Mesh Reader
   )
   /SetVariables(
    Particles.MeshOrientation=&MultiplyQuaternion(
     Quaternion A=&LerpQuaternion(
      Quaternion A:Quat4f=$Particles.Initial.MeshOrientation
      Quaternion B:Quat4f=&MakeQuaternion(
       Quaternion From=0
       Axis=<0 0 1>
       Angle=0
       From=?
       To=?
       Stabilization Mode:ENiagara_OrientVectorCount=::NewEnumerator0
       Side Direction:Vector3f=?
       Up Direction:Vector3f=?
       Angle Type=::NewEnumerator0
       Normalize:NiagaraBool=true
      )
      Lerp Factor:NiagaraFloat=&FloatFromCurve(FloatCurve=DataInterface() {propertyValues {Curve {defaultValue=1, keys [<at 0 0>,<at 0.2 1>]}}})
      Enable Spherical Linear Interpolation:NiagaraBool=false
     )
     Quaternion B=&MakeQuaternion(
      Quaternion From=0
      Axis=<0 0 1>
      Angle=&Multiply_Float(A=300, B=$Engine.System.Age)
      From=?
      To=?
      Stabilization Mode:ENiagara_OrientVectorCount=::NewEnumerator0
      Side Direction:Vector3f=?
      Up Direction:Vector3f=?
      Angle Type=::NewEnumerator0
      Normalize:NiagaraBool=true
     )
    )
    Particles.PositionRings:NiagaraPosition=$Particles.SkeletalMeshLocation.SampledPosition
   ) {isSetParameters=true, module=SetVariables_#6}
  ]
  rendererClasses [&NiagaraMeshRendererProperties, &NiagaraSpriteRendererProperties]
  renderers [
   {rendererClass=&NiagaraMeshRendererProperties, rendererIndex=0}
   {rendererClass=&NiagaraMeshRendererProperties, rendererIndex=1}
   {rendererClass=&NiagaraSpriteRendererProperties, rendererIndex=2}
   {rendererClass=&NiagaraSpriteRendererProperties, rendererIndex=3}
   {rendererClass=&NiagaraSpriteRendererProperties, rendererIndex=4}
   {rendererClass=&NiagaraSpriteRendererProperties, rendererIndex=5}
  ]
 }
 Rings_Toon_High {
  emitterUpdateScript [
   &EmitterState(
    Life Cycle Mode=::NewEnumerator1
    Loop Behavior=::NewEnumerator2
    Loop Count=7
    Loop Duration=0.25
    Recalculate Duration Each Loop=true
    UseLoopDelay=false
   )
   &SpawnBurst_Instantaneous(Use Spawn Probability=false, UseLoopCountLimit=false) {module=SpawnBurst_Instantaneous001}
   &SpawnBurst_Instantaneous(Spawn Group=1, Use Spawn Probability=false, UseLoopCountLimit=false) {module=SpawnBurst_Instantaneous002}
   &SpawnBurst_Instantaneous(Spawn Group=2, Use Spawn Probability=false, UseLoopCountLimit=false) {module=SpawnBurst_Instantaneous003}
   /SetVariables(Emitter.Emissive_Intensity:NiagaraFloat=&FNTimeOfDayScalar(Night Scale=0.4, Morning Scale=0.225, Day Scale=0.15, Evening Scale=0.25)) {isSetParameters=true}
   &SpawnBurst_Instantaneous(Loop Count Limit=5, Use Spawn Probability=false, UseLoopCountLimit=false)
  ]
  particleSpawnScript [
   /SetVariables(Particles.PartitionParticles.Partition:NiagaraInt32=$Emitter.LoopCount, Particles.VisibilityTag=&SpawnGroup()) {isSetParameters=true, module=SetVariables_#3}
   &InitializeParticle(
    Color=&MergeHelper_ColorSwitch(
     0=<rgba 2.3125 1.1562 0 1>
     1=<rgba 3 1.3809 0.2244 1>
     2=<rgba 3 1.3812 0.225 1>
     3=&MakeLinearColorFromFloat(Value:NiagaraFloat=&Lerp_Float(
       A=5
       B=50
       Alpha=&FloatFromCurve(
        FloatCurve=DataInterface() {propertyValues {Curve {keys [
            {interpMode=RCIM_Cubic, tangentMode=RCTM_User, time=0, value=0.2}
            {arriveTangent=1.1314, interpMode=RCIM_Cubic, leaveTangent=1.1314, tangentMode=RCTM_User, time=0.9, value=0.6}
            {interpMode=RCIM_Cubic, tangentMode=RCTM_User, time=1, value=1}
           ]}}}
        CurveIndex=&RandomRangeFloat(Minimum=0, Maximum=1)
       )
      ))
     4=&MakeLinearColorFromFloat(Value:NiagaraFloat=&Lerp_Float(
       A=5
       B=50
       Alpha=&FloatFromCurve(
        FloatCurve=DataInterface() {propertyValues {Curve {keys [
            {interpMode=RCIM_Cubic, tangentMode=RCTM_User, time=0, value=0.2}
            {arriveTangent=1.1314, interpMode=RCIM_Cubic, leaveTangent=1.1314, tangentMode=RCTM_User, time=0.9, value=0.6}
            {interpMode=RCIM_Cubic, tangentMode=RCTM_User, time=1, value=1}
           ]}}}
        CurveIndex=&RandomRangeFloat(Minimum=0, Maximum=1)
       )
      ))
     Index=$Particles.VisibilityTag
     Num=::NewEnumerator1
    )
    Lifetime=&MergeHelper_FloatSwitch(
     0=&MergeHelper_FloatSwitch(0=4.948, 1=4.552, 2=4.034, 3=3.517, 4=3, 5=2.627, 6=2.167, Index=$Particles.PartitionParticles.Partition, Num=::NewEnumerator5)
     1=0.1
     2=0.1
     3=0.1
     4=&RandomRangeFloat(Minimum=0.3, Maximum=0.6)
     Index=$Particles.VisibilityTag
     Num=::NewEnumerator2
    )
    Mesh Scale Mode=::NewEnumerator3
    Mesh Uniform Scale=1.2
    Sprite Rotation Angle=&RandomRangeFloat(Minimum=-45, Maximum=45)
    Sprite Rotation Mode=::NewEnumerator0
    Sprite Size=&Multiply_Vector2D_ByFloat(Float=&RandomRangeFloat(Minimum=0.45, Maximum=0.55), Vector2D=<35 50>)
    Sprite Size Mode=::NewEnumerator4
    Sprite UV Mode=::NewEnumerator1
    UsePositionOffset=false
    Write Material Random=true
   )
   &InitialMeshOrientation(Mesh Orientation Mode=::NewEnumerator6, Orientation Quaternion:Quat4f=<1 0 0 1>, Use Rotation Vector=false)
   &SkeletalMeshLocation(
    Bone / Socket Index=$Particles.PartitionParticles.Partition
    Bone Sampling Mode=::NewEnumerator4
    MaskSpawnGroup=false
    Orientation Sampling=::NewEnumerator1
    Sampled Position Offset=&MergeHelper_VectorSwitch(0=<0 0 0>, 1=<0 0 2>, 2=<0 0 2>, 3=?, Index=$Particles.VisibilityTag, Num=::NewEnumerator1)
    Sampled Position Offset Coordinate Space:ENiagaraSkelMeshTransforms=::NewEnumerator0
    Skeletal Mesh=$System.Skeletal Mesh Reader
   ) {module=SkeletalMeshLocation001}
   /SetVariables(Particles.GlitterSubUV:NiagaraFloat=&RandomRangeFloat(Minimum=0, Maximum=3)) {isSetParameters=true, module=SetVariables_#4}
  ]
  particleUpdateScript [
   &ParticleState()
   &ScaleMeshSize(Scale Factor=&VectorFromCurve(VectorCurve=DataInterface() {propertyValues {
       XCurve {keys [{arriveTangent=13.4177, interpMode=RCIM_Cubic, leaveTangent=13.4177, tangentMode=RCTM_User, time=0, value=0}, <at 0.15 1 RCIM_Cubic>]}
       YCurve {keys [<at 0 1>,<at 1 1>]}
       ZCurve {keys [{arriveTangent=13.4177, interpMode=RCIM_Cubic, leaveTangent=13.4177, tangentMode=RCTM_User, time=0, value=0}, <at 0.15 1 RCIM_Cubic>]}
      }}))
   &SkeletalMeshLocation(
    Bone / Socket Index=$Particles.PartitionParticles.Partition
    Bone Sampling Mode=::NewEnumerator4
    MaskSpawnGroup=false
    Orientation Sampling=::NewEnumerator1
    Sampled Position Offset=&MergeHelper_VectorSwitch(0=<0 0 0>, 1=<0 0 3>, 2=<0 0 3>, 3=?, Index=$Particles.VisibilityTag, Num=::NewEnumerator1)
    Sampled Position Offset Coordinate Space:ENiagaraSkelMeshTransforms=::NewEnumerator0
    Skeletal Mesh=$System.Skeletal Mesh Reader
   )
   &Drag(Use Linear Drag=true, Use Rotational Drag=false)
   &ScaleSpriteSize(
    Scale Factor=&Vector2DFromCurve(
     Vector2Curve=DataInterface() {propertyValues {
       XCurve {
        defaultValue=0
        keys [
         {arriveTangent=0.0865, interpMode=RCIM_Cubic, leaveTangent=0.0865, tangentMode=RCTM_User, time=0, value=0}
         {arriveTangent=4.9825, interpMode=RCIM_Cubic, leaveTangent=4.9825, tangentMode=RCTM_User, time=0.4724, value=1}
        ]
       }
       YCurve {
        defaultValue=1
        keys [
         {arriveTangent=-5.5781, interpMode=RCIM_Cubic, leaveTangent=-5.5781, tangentMode=RCTM_User, time=0.6, value=1.1}
         {arriveTangent=-0.1643, interpMode=RCIM_Cubic, leaveTangent=-0.1643, tangentMode=RCTM_User, time=1, value=0}
        ]
       }
      }}
     ScaleCurve=&RandomRangeVector2D(Minimum=<1 1>, Maximum=<1 1.2>)
    )
    Scale Sprite Size Mode=::NewEnumerator2
   ) {module=ScaleSpriteSize001}
   /SetVariables(
    Particles.CameraOffsetStar=20
    Particles.GlowColor=&MakeLinearColorFromVectorAndFloat(
     "Vector (RGB)"=&Multiply_Vector(
      A=&MakeVectorFromLinearColorRGB(Color=<rgba 1.2 0.6795 0 1>)
      B=&MakeVectorFromLinearColorRGB(Color=$System.TOD)
     )
     "Float (Alpha)"=1
    )
    Particles.GlowSize=&Multiply_Vector2D_ByFloat(Float=1.3, Vector2D=$Particles.SpriteSize)
    Particles.StarPosition:NiagaraPosition=&AddVectorToPosition(Position=$Particles.Initial.Position, Vector=<0 0 0>)
   ) {isSetParameters=true, module=SetVariables_#5}
   &SolveForcesAndVelocity(Clamp Velocity=false, Limit Acceleration=false)
   /SetVariables(Particles.MeshOrientation=&MultiplyQuaternion(
     Quaternion A=&LerpQuaternion(
      Quaternion A:Quat4f=$Particles.Initial.MeshOrientation
      Quaternion B:Quat4f=&MakeQuaternion(
       Quaternion From=0
       Axis=<0 0 1>
       Angle=0
       From=?
       To=?
       Stabilization Mode:ENiagara_OrientVectorCount=::NewEnumerator0
       Side Direction:Vector3f=?
       Up Direction:Vector3f=?
       Angle Type=::NewEnumerator0
       Normalize:NiagaraBool=true
      )
      Lerp Factor:NiagaraFloat=&FloatFromCurve(FloatCurve=DataInterface() {propertyValues {Curve {defaultValue=1, keys [<at 0 0>,<at 0.2 1>]}}})
      Enable Spherical Linear Interpolation:NiagaraBool=false
     )
     Quaternion B=&MakeQuaternion(
      Quaternion From=0
      Axis=<0 0 1>
      Angle=&Multiply_Float(A=300, B=$Engine.System.Age)
      From=?
      To=?
      Stabilization Mode:ENiagara_OrientVectorCount=::NewEnumerator0
      Side Direction:Vector3f=?
      Up Direction:Vector3f=?
      Angle Type=::NewEnumerator0
      Normalize:NiagaraBool=true
     )
    )) {isSetParameters=true, module=SetVariables_#6}
   /SetVariables(Particles.MeshOrientation=&LerpQuaternion(
     Quaternion A:Quat4f=$Particles.Initial.MeshOrientation
     Quaternion B:Quat4f=&MakeQuaternion(
      Quaternion From=0
      Axis=<0 0 1>
      Angle=&Multiply_Float(A=300, B=$Engine.System.Age)
      From=?
      To=?
      Stabilization Mode:ENiagara_OrientVectorCount=::NewEnumerator0
      Side Direction:Vector3f=?
      Up Direction:Vector3f=?
      Angle Type=::NewEnumerator0
      Normalize:NiagaraBool=true
     )
     Lerp Factor:NiagaraFloat=&MaskFloat(
      Mask Source=1
      Mask Parameter=$Particles.PartitionParticles.Partition
      Mask=6
      Masked Float=&FloatFromCurve(FloatCurve=DataInterface() {propertyValues {Curve {defaultValue=1, keys [<at 0 0>,<at 0.125 1>]}}})
      Passthrough Float=&FloatFromCurve(FloatCurve=DataInterface() {propertyValues {Curve {defaultValue=1, keys [<at 0 0>,<at 0.2 1>]}}})
     )
     Enable Spherical Linear Interpolation:NiagaraBool=false
    )) {enabled=false, isSetParameters=true, module=SetVariables_#7}
   /SetVariables(
    Particles.FlareColor=&MakeLinearColorFromVectorAndFloat("Vector (RGB)"=&MakeVectorFromLinearColorRGB(Color=<rgba 3 1.7047 0.2244 1>), "Float (Alpha)"=0.15)
    Particles.FlareSize:Vector2f=<100 100>
   ) {isSetParameters=true, module=SetVariables_#8}
   &ScaleColor(
    Scale RGB=&MergeHelper_VectorSwitch(
     0=&FN_FrontEndVector3Modifier(
      Gameplay=&FN_Select_Vector_from_Time_Of_Day(Morning=<1.115 1.15 1.15>, Day=<1 1 1>, Evening=<1 1 1>, Night=<3.5 2.4 1>)
      Front End=<3.2 2.75 2.75>
     )
     1=<1 1 1>
     2=<1 1 1>
     3=?
     Index=$Particles.VisibilityTag
     Num=::NewEnumerator1
    )
    ScaleA=true
    ScaleRGB=true
   )
  ]
  rendererClasses [&NiagaraMeshRendererProperties, &NiagaraSpriteRendererProperties]
  renderers [
   {rendererClass=&NiagaraMeshRendererProperties, rendererIndex=0}
   {rendererClass=&NiagaraMeshRendererProperties, rendererIndex=1}
   {rendererClass=&NiagaraSpriteRendererProperties, rendererIndex=2}
   {rendererClass=&NiagaraSpriteRendererProperties, rendererIndex=3}
   {rendererClass=&NiagaraSpriteRendererProperties, rendererIndex=4}
  ]
 }
 Rings_Toon_High001 {
  bEnabled=false
  emitterUpdateScript [
   &EmitterState(
    Life Cycle Mode=::NewEnumerator1
    Loop Behavior=::NewEnumerator2
    Loop Count=7
    Loop Duration=0.25
    Recalculate Duration Each Loop=true
    UseLoopDelay=false
   )
   &SpawnBurst_Instantaneous(Loop Count Limit=5, Use Spawn Probability=false, UseLoopCountLimit=false)
   &SpawnBurst_Instantaneous(Use Spawn Probability=false, UseLoopCountLimit=false) {enabled=false, module=SpawnBurst_Instantaneous001}
   &SpawnBurst_Instantaneous(Spawn Group=1, Use Spawn Probability=false, UseLoopCountLimit=false) {module=SpawnBurst_Instantaneous002}
   &SpawnBurst_Instantaneous(Spawn Group=2, Use Spawn Probability=false, UseLoopCountLimit=false) {module=SpawnBurst_Instantaneous003}
  ]
  particleSpawnScript [
   /SetVariables(Particles.PartitionParticles.Partition:NiagaraInt32=$Emitter.LoopCount, Particles.VisibilityTag=&SpawnGroup()) {isSetParameters=true, module=SetVariables_#3}
   &InitializeParticle(
    Color=&MergeHelper_ColorSwitch(
     0=<rgba 2.3125 1.1562 0 1>
     1=<rgba 3 1.3809 0.2244 1>
     2=<rgba 3 1.3812 0.225 1>
     3=&MakeLinearColorFromFloat(Value:NiagaraFloat=&Lerp_Float(
       A=5
       B=50
       Alpha=&FloatFromCurve(
        FloatCurve=DataInterface() {propertyValues {Curve {keys [
            {interpMode=RCIM_Cubic, tangentMode=RCTM_User, time=0, value=0.2}
            {arriveTangent=1.1314, interpMode=RCIM_Cubic, leaveTangent=1.1314, tangentMode=RCTM_User, time=0.9, value=0.6}
            {interpMode=RCIM_Cubic, tangentMode=RCTM_User, time=1, value=1}
           ]}}}
        CurveIndex=&RandomRangeFloat(Minimum=0, Maximum=1)
       )
      ))
     4=&MakeLinearColorFromFloat(Value:NiagaraFloat=&Lerp_Float(
       A=5
       B=50
       Alpha=&FloatFromCurve(
        FloatCurve=DataInterface() {propertyValues {Curve {keys [
            {interpMode=RCIM_Cubic, tangentMode=RCTM_User, time=0, value=0.2}
            {arriveTangent=1.1314, interpMode=RCIM_Cubic, leaveTangent=1.1314, tangentMode=RCTM_User, time=0.9, value=0.6}
            {interpMode=RCIM_Cubic, tangentMode=RCTM_User, time=1, value=1}
           ]}}}
        CurveIndex=&RandomRangeFloat(Minimum=0, Maximum=1)
       )
      ))
     Index=$Particles.VisibilityTag
     Num=::NewEnumerator1
    )
    Lifetime=&MergeHelper_FloatSwitch(
     0=&MergeHelper_FloatSwitch(0=4.948, 1=4.552, 2=4.034, 3=3.517, 4=3, 5=2.627, 6=2.167, Index=$Particles.PartitionParticles.Partition, Num=::NewEnumerator5)
     1=0.1
     2=0.1
     3=0.1
     4=&RandomRangeFloat(Minimum=0.3, Maximum=0.6)
     Index=$Particles.VisibilityTag
     Num=::NewEnumerator2
    )
    Mesh Scale Mode=::NewEnumerator3
    Mesh Uniform Scale=1.2
    Sprite Rotation Angle=&RandomRangeFloat(Minimum=-45, Maximum=45)
    Sprite Rotation Mode=::NewEnumerator0
    Sprite Size=&Multiply_Vector2D_ByFloat(Float=&RandomRangeFloat(Minimum=0.45, Maximum=0.55), Vector2D=<35 50>)
    Sprite Size Mode=::NewEnumerator4
    Sprite UV Mode=::NewEnumerator1
    UsePositionOffset=false
    Write Material Random=true
   )
   &InitialMeshOrientation(Mesh Orientation Mode=::NewEnumerator6, Orientation Quaternion:Quat4f=<1 0 0 1>, Use Rotation Vector=false)
   &SkeletalMeshLocation(
    Bone / Socket Index=$Particles.PartitionParticles.Partition
    Bone Sampling Mode=::NewEnumerator4
    MaskSpawnGroup=false
    Orientation Sampling=::NewEnumerator1
    Sampled Position Offset=&MergeHelper_VectorSwitch(0=<0 0 0>, 1=<0 0 2>, 2=<0 0 2>, 3=?, Index=$Particles.VisibilityTag, Num=::NewEnumerator1)
    Sampled Position Offset Coordinate Space:ENiagaraSkelMeshTransforms=::NewEnumerator0
    Skeletal Mesh=$System.Skeletal Mesh Reader
   ) {module=SkeletalMeshLocation001}
   /SetVariables(Particles.GlitterSubUV:NiagaraFloat=&RandomRangeFloat(Minimum=0, Maximum=3)) {isSetParameters=true, module=SetVariables_#4}
  ]
  particleUpdateScript [
   &ParticleState()
   &ScaleMeshSize(Scale Factor=&VectorFromCurve(VectorCurve=DataInterface() {propertyValues {
       XCurve {keys [{arriveTangent=13.4177, interpMode=RCIM_Cubic, leaveTangent=13.4177, tangentMode=RCTM_User, time=0, value=0}, <at 0.15 1 RCIM_Cubic>]}
       YCurve {keys [<at 0 1>,<at 1 1>]}
       ZCurve {keys [{arriveTangent=13.4177, interpMode=RCIM_Cubic, leaveTangent=13.4177, tangentMode=RCTM_User, time=0, value=0}, <at 0.15 1 RCIM_Cubic>]}
      }}))
   &SkeletalMeshLocation(
    Bone / Socket Index=$Particles.PartitionParticles.Partition
    Bone Sampling Mode=::NewEnumerator4
    MaskSpawnGroup=false
    Orientation Sampling=::NewEnumerator1
    Sampled Position Offset=&MergeHelper_VectorSwitch(0=<0 0 0>, 1=<0 0 3>, 2=<0 0 3>, 3=?, Index=$Particles.VisibilityTag, Num=::NewEnumerator1)
    Sampled Position Offset Coordinate Space:ENiagaraSkelMeshTransforms=::NewEnumerator0
    Skeletal Mesh=$System.Skeletal Mesh Reader
   )
   &Drag(Use Linear Drag=true, Use Rotational Drag=false)
   &ScaleSpriteSize(
    Scale Factor=&Vector2DFromCurve(
     Vector2Curve=DataInterface() {propertyValues {
       XCurve {
        defaultValue=0
        keys [
         {arriveTangent=0.0865, interpMode=RCIM_Cubic, leaveTangent=0.0865, tangentMode=RCTM_User, time=0, value=0}
         {arriveTangent=4.9825, interpMode=RCIM_Cubic, leaveTangent=4.9825, tangentMode=RCTM_User, time=0.4724, value=1}
        ]
       }
       YCurve {
        defaultValue=1
        keys [
         {arriveTangent=-5.5781, interpMode=RCIM_Cubic, leaveTangent=-5.5781, tangentMode=RCTM_User, time=0.6, value=1.1}
         {arriveTangent=-0.1643, interpMode=RCIM_Cubic, leaveTangent=-0.1643, tangentMode=RCTM_User, time=1, value=0}
        ]
       }
      }}
     ScaleCurve=&RandomRangeVector2D(Minimum=<1 1>, Maximum=<1 1.2>)
    )
    Scale Sprite Size Mode=::NewEnumerator2
   ) {module=ScaleSpriteSize001}
   /SetVariables(
    Particles.CameraOffsetStar=20
    Particles.GlowColor=&MakeLinearColorFromVectorAndFloat(
     "Vector (RGB)"=&Multiply_Vector(
      A=&MakeVectorFromLinearColorRGB(Color=<rgba 1.2 0.6795 0 1>)
      B=&MakeVectorFromLinearColorRGB(Color=$System.TOD)
     )
     "Float (Alpha)"=1
    )
    Particles.GlowSize=&Multiply_Vector2D_ByFloat(Float=1.3, Vector2D=$Particles.SpriteSize)
    Particles.StarPosition:NiagaraPosition=&AddVectorToPosition(Position=$Particles.Initial.Position, Vector=<0 0 0>)
   ) {isSetParameters=true, module=SetVariables_#5}
   &SolveForcesAndVelocity(Clamp Velocity=false, Limit Acceleration=false)
   /SetVariables(Particles.MeshOrientation=&MultiplyQuaternion(
     Quaternion A=&LerpQuaternion(
      Quaternion A:Quat4f=$Particles.Initial.MeshOrientation
      Quaternion B:Quat4f=&MakeQuaternion(
       Quaternion From=0
       Axis=<0 0 1>
       Angle=0
       From=?
       To=?
       Stabilization Mode:ENiagara_OrientVectorCount=::NewEnumerator0
       Side Direction:Vector3f=?
       Up Direction:Vector3f=?
       Angle Type=::NewEnumerator0
       Normalize:NiagaraBool=true
      )
      Lerp Factor:NiagaraFloat=&FloatFromCurve(FloatCurve=DataInterface() {propertyValues {Curve {defaultValue=1, keys [<at 0 0>,<at 0.2 1>]}}})
      Enable Spherical Linear Interpolation:NiagaraBool=false
     )
     Quaternion B=&MakeQuaternion(
      Quaternion From=0
      Axis=<0 0 1>
      Angle=&Multiply_Float(A=300, B=$Engine.System.Age)
      From=?
      To=?
      Stabilization Mode:ENiagara_OrientVectorCount=::NewEnumerator0
      Side Direction:Vector3f=?
      Up Direction:Vector3f=?
      Angle Type=::NewEnumerator0
      Normalize:NiagaraBool=true
     )
    )) {isSetParameters=true, module=SetVariables_#6}
   /SetVariables(Particles.MeshOrientation=&LerpQuaternion(
     Quaternion A:Quat4f=$Particles.Initial.MeshOrientation
     Quaternion B:Quat4f=&MakeQuaternion(
      Quaternion From=0
      Axis=<0 0 1>
      Angle=&Multiply_Float(A=300, B=$Engine.System.Age)
      From=?
      To=?
      Stabilization Mode:ENiagara_OrientVectorCount=::NewEnumerator0
      Side Direction:Vector3f=?
      Up Direction:Vector3f=?
      Angle Type=::NewEnumerator0
      Normalize:NiagaraBool=true
     )
     Lerp Factor:NiagaraFloat=&MaskFloat(
      Mask Source=1
      Mask Parameter=$Particles.PartitionParticles.Partition
      Mask=6
      Masked Float=&FloatFromCurve(FloatCurve=DataInterface() {propertyValues {Curve {defaultValue=1, keys [<at 0 0>,<at 0.125 1>]}}})
      Passthrough Float=&FloatFromCurve(FloatCurve=DataInterface() {propertyValues {Curve {defaultValue=1, keys [<at 0 0>,<at 0.2 1>]}}})
     )
     Enable Spherical Linear Interpolation:NiagaraBool=false
    )) {enabled=false, isSetParameters=true, module=SetVariables_#7}
   /SetVariables(
    Particles.FlareColor=&MakeLinearColorFromVectorAndFloat("Vector (RGB)"=&MakeVectorFromLinearColorRGB(Color=<rgba 3 1.7047 0.2244 1>), "Float (Alpha)"=0.15)
    Particles.FlareSize:Vector2f=<100 100>
   ) {isSetParameters=true, module=SetVariables_#8}
   &ScaleColor(
    Scale RGB=&MergeHelper_VectorSwitch(
     0=&FN_FrontEndVector3Modifier(
      Gameplay=&FN_Select_Vector_from_Time_Of_Day(Morning=<1.115 1.15 1.15>, Day=<1 1 1>, Evening=<1 1 1>, Night=<3.5 2.4 1>)
      Front End=<3.2 2.75 2.75>
     )
     1=<1 1 1>
     2=<1 1 1>
     3=?
     Index=$Particles.VisibilityTag
     Num=::NewEnumerator1
    )
    ScaleA=true
    ScaleRGB=true
   )
  ]
  rendererClasses [&NiagaraMeshRendererProperties, &NiagaraSpriteRendererProperties]
  renderers [
   {rendererClass=&NiagaraMeshRendererProperties, rendererIndex=0}
   {rendererClass=&NiagaraMeshRendererProperties, rendererIndex=1}
   {rendererClass=&NiagaraSpriteRendererProperties, rendererIndex=2}
   {rendererClass=&NiagaraSpriteRendererProperties, rendererIndex=3}
   {rendererClass=&NiagaraSpriteRendererProperties, rendererIndex=4}
  ]
 }
}
summary {
 systemName=NS_CompanionEmote_SizeTribe_Rings
 emitters [
  {emitterName=GlitterFalling001, rendererClasses [&NiagaraSpriteRendererProperties]}
  {emitterName=Rings_Toon2, rendererClasses [&NiagaraMeshRendererProperties, &NiagaraSpriteRendererProperties]}
  {emitterName=GlitterSpawn001, rendererClasses [&NiagaraSpriteRendererProperties]}
  {
   emitterName=Rings_Toon_High
   rendererClasses [&NiagaraMeshRendererProperties, &NiagaraSpriteRendererProperties]
  }
  {
   bEnabled=false
   emitterName=Rings_Toon_High001
   rendererClasses [&NiagaraMeshRendererProperties, &NiagaraSpriteRendererProperties]
  }
 ]
 userVariables [
  User.bIsFrontEnd {type=NiagaraBool}
  User.bIsFrontEndPreview {type=NiagaraBool}
  User.FirstTime {type=NiagaraBool}
  User.ImpactLocation {type=NiagaraPosition, description="", defaultValue {}}
  User.NormalizedNotifyProgress {type=NiagaraFloat, defaultValue=0, description=""}
  User.ParticleScale {type=NiagaraFloat, defaultValue=1, description=""}
  User.Returning {type=NiagaraBool}
 ]
}
