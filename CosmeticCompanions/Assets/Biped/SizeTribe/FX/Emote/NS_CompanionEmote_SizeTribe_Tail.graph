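# Niagara effect for the SizeTribe companion tail emote, drawing a spinning smear twirl with wind, ball, dust and speed-line elements.
# One Twirl emitter spawns seven groups, mostly driven by user rates, with a VisibilityTag selecting each group's look and motion.
# Particle updates resample skeletal mesh bones and apply age-driven spin, writing separate position, orientation, scale and material bindings per element for multiple mesh and sprite renderers.
# Colors and intensity adjust for time of day and front end.
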NiagaraSystem /CosmeticCompanions/Assets/Biped/SizeTribe/FX/Emote/NS_CompanionEmote_SizeTribe_Tail
dependencies {
 usedDataInterfaces [
  &NiagaraDataInterfaceArrayFloat3
  &NiagaraDataInterfaceCurve
  &NiagaraDataInterfaceVectorCurve
  &NiagaraDataInterfaceSkeletalMesh
 ]
 usedDynamicInputs [&MakeFloatFromVector]
 usedModules [
  &self:SystemSpawnScript.SystemScriptSource.SystemScriptGraph.NiagaraNodeAssignment_0.SetVariables_#1
  &self:SystemSpawnScript.SystemScriptSource.SystemScriptGraph.NiagaraNodeAssignment_1.SetVariables_#2
  &self:SystemSpawnScript.SystemScriptSource.SystemScriptGraph.NiagaraNodeAssignment_2.SetVariables_#3
  &self:SystemSpawnScript.SystemScriptSource.SystemScriptGraph.NiagaraNodeAssignment_3.SetVariables_#4
  &self:SystemSpawnScript.SystemScriptSource.SystemScriptGraph.NiagaraNodeAssignment_4.SetVariables_#5
  &SystemState
 ]
}
emitters {Twirl {
  emitterUpdateScript [
   &EmitterState(Life Cycle Mode=::NewEnumerator1, Loop Behavior=::NewEnumerator1, Loop Delay=1.089, Loop Duration=6.627, UseLoopDelay=false)
   &SpawnRate(
    SpawnRate=&Multiply_Float(
     A=&Multiply_Float(A=$User.TwirlSpawnRate, B=2)
     B=&RandomRangeFloat(Minimum=1, Maximum=1)
    )
    Use Spawn Probability=false
   )
   &SpawnRate(
    SpawnGroup=1
    SpawnRate=&Multiply_Float(
     A=&Multiply_Float(A=$User.TwirlSpawnRate, B=2)
     B=&RandomRangeFloat(Minimum=1, Maximum=1)
    )
    Use Spawn Probability=false
   ) {module=SpawnRate001}
   &SpawnRate(
    SpawnGroup=2
    SpawnRate=&Multiply_Float(
     A=&Multiply_Float(A=$User.TwirlSpawnRate, B=2)
     B=&RandomRangeFloat(Minimum=1, Maximum=1)
    )
    Use Spawn Probability=false
   ) {module=SpawnRate002}
   &SpawnRate(
    Spawn Probability=&Divide_Float(A=$User.TwirlSpawnRate, B=4)
    SpawnGroup=3
    SpawnRate=10
    Use Spawn Probability=true
   ) {module=SpawnRate003}
   &SpawnRate(Spawn Probability=$User.SpawnRateBall, SpawnGroup=4, SpawnRate=10, Use Spawn Probability=true) {module=SpawnRate004}
   &SpawnBurst_Instantaneous(
    Spawn Count=&RandomRangeInt(Minimum=5, Maximum=8)
    Spawn Group=5
    Spawn Time=0.5
    Use Spawn Probability=false
    UseLoopCountLimit=false
   )
   &SpawnRate(
    Spawn Probability=$User.SpawnRateSpeedLines
    SpawnGroup=6
    SpawnRate=&FN_FrontEndFloatModifier(Gameplay=100, Front End=150)
    Use Spawn Probability=true
   ) {module=SpawnRate005}
  ]
  particleSpawnScript [
   /SetVariables(Particles.VisibilityTag=&SpawnGroup()) {isSetParameters=true}
   &PartitionParticles(
    Number Of Partitions=&MergeHelper_IntSwitch(0=0, 1=0, 2=0, 3=3, 4=3, 5=0, Index=$Particles.VisibilityTag, Num=::NewEnumerator4)
    Partition Mode:ENiagara_PartitionMode_v2=::NewEnumerator5
   )
   &InitializeParticle(
    Color=&MergeHelper_ColorSwitch(
     0=&RandomRangeLinearColor(Minimum=<rgba 0.6208 0.6726 0.776 1>, Maximum=<rgba 0.45 0.4875 0.5625 1>)
     1=&RandomRangeLinearColor(Minimum=<rgba 0.6208 0.6726 0.776 1>, Maximum=<rgba 0.45 0.4875 0.5625 1>)
     2=&RandomRangeLinearColor(Minimum=<rgba 0.6208 0.6726 0.776 1>, Maximum=<rgba 0.45 0.4875 0.5625 1>)
     3=<rgba 2 2 2 1>
     4=$System.Inner_Smear_Color
     5=<rgba 1 1 1 0.5>
     6=$System.Inner_Smear_Color
     Index=$Particles.VisibilityTag
     Num=::NewEnumerator5
    )
    Lifetime=&MergeHelper_FloatSwitch(
     0=&RandomRangeFloat(Minimum=0.1, Maximum=0.2)
     1=&RandomRangeFloat(Minimum=0.1, Maximum=0.2)
     2=&RandomRangeFloat(Minimum=0.1, Maximum=0.2)
     3=&RandomRangeFloat(Minimum=0.1, Maximum=0.2)
     4=&RandomRangeFloat(Minimum=0.2, Maximum=0.4)
     5=&RandomRangeFloat(Minimum=0.3, Maximum=0.5)
     6=&RandomRangeFloat(Minimum=0.05, Maximum=0.1)
     Index=$Particles.VisibilityTag
     Num=::NewEnumerator5
    )
    Mesh Scale=&MergeHelper_VectorSwitch(
     0=<0.25 0.25 0.4>
     1=<0.35 0.35 0.4>
     2=<0.45 0.45 0.4>
     3=<0.47 0.47 0.4>
     4=&MergeHelper_VectorSwitch(0=<0.34 0.34 0.34>, 1=<0.36 0.36 0.36>, 2=<0.38 0.4 0.38>, 3=?, Index=$Particles.PartitionParticles.Partition, Num=::NewEnumerator1)
     5=&VectorFromFloat(Value=&RandomRangeFloat(Minimum=0.1, Maximum=0.4))
     Index=$Particles.VisibilityTag
     Num=::NewEnumerator4
    )
    Mesh Scale Mode=::NewEnumerator4
    Sprite Size Max=<30 115>
    Sprite Size Min=<3 41>
    Sprite Size Mode=::NewEnumerator2
    Sprite UV Mode=::NewEnumerator3
    UsePositionOffset=false
    Write Material Random=true
   )
   &CameraOffset(Camera Offset Amount=&FN_FrontEndFloatModifier(Gameplay=15, Front End=5)) {enabled=false}
   /SetVariables(Particles.Color=&MultiplyLinearColor(
     Linear Color=&MergeHelper_ColorSwitch(
      0=$System.Inner_Smear_Color
      1=$System.Inner_Smear_Color
      2=&FNTimeOfDayColor(
       Night Color=<rgba 2.5 2.5 2.5 0.2>
       Morning Color=<rgba 2.5 2.5 2.5 0.3>
       Day Color=<rgba 2.5 2.5 2.5 0.4>
       Evening Color=<rgba 2.5 2.5 2.5 0.3>
      )
      3=&FNTimeOfDayColor(
       Night Color=<rgba 2.5 2.5 2.5 0.2>
       Morning Color=<rgba 2.5 2.5 2.5 0.3>
       Day Color=<rgba 2.5 2.5 2.5 0.4>
       Evening Color=<rgba 2.5 2.5 2.5 0.3>
      )
      4=&Multiply_LinearColorByFloat(Linear Color=$System.Inner_Smear_Color, Float=1.25)
      5=<rgba 1 1 1 0.5>
      6=$System.Inner_Smear_Color
      Index=$Particles.VisibilityTag
      Num=::NewEnumerator5
     )
     Scale Factor=&MakeVector4_FromLinearColor(LinearColor=$System.TOD Color)
    )) {isSetParameters=true}
   &InitialMeshOrientation(
    Mesh Axis To Orient=&MergeHelper_VectorSwitch(0=<0 0 1>, 1=<0 0 1>, 2=<0 0 1>, 3=<0 0 1>, 4=<0 0 1>, 5=<1 0 0>, Index=$Particles.VisibilityTag, Num=::NewEnumerator4)
    Mesh Orientation Mode=::NewEnumerator2
    Orientation Vector=&MergeHelper_VectorSwitch(0=<0 0 1>, 1=<0 0 1>, 2=<0 0 1>, 3=<0 0 1>, 4=<1 0 0>, 5=<1 0 0>, Index=$Particles.VisibilityTag, Num=::NewEnumerator4)
    Rotation=&RandomRangeVector(Minimum=<0 0 -1>, Maximum=<0 0 1>)
    Use Rotation Vector=true
   ) {module=InitialMeshOrientation001}
   &SkeletalMeshLocation(
    Bone Sampling Mode=::NewEnumerator4
    MaskSpawnGroup=false
    Orientation Sampling=::NewEnumerator1
    Sampled Position Offset=&MergeHelper_VectorSwitch(0=<0 0 0>, 1=<0 0 0>, 2=<0 0 0>, 3=<0 0 0>, 4=<0 0 0>, 5=<0 0 0>, 6=<0 0 50>, Index=$Particles.VisibilityTag, Num=::NewEnumerator5)
    Skeletal Mesh=$System.Skeletal Mesh
   ) {module=SkeletalMeshLocation003}
   &ShapeLocation(
    Box Size=&MergeHelper_VectorSwitch(0=<0 0 0>, 1=<0 0 0>, 2=<0 0 0>, 3=<0 0 0>, 4=<0 0 0>, 5=<0 0 0>, 6=<30 30 40>, Index=$Particles.VisibilityTag, Num=::NewEnumerator5)
    MaskSpawnGroup=false
    Override Seed=false
    Shape Origin=&MaskVector(
     Mask Source=1
     Mask Parameter=5
     Mask=$Particles.VisibilityTag
     Masked Vector=&AddVectorToPosition(
      Position=$Output.SkeletalMeshLocation003.SampledPositionWithoutOffset
      Vector=&FN_FrontEndVector3Modifier(Gameplay=<0 0 -40>, Front End=<0 0 -35>)
     )
     Passthrough Vector=Linked() {linkedVariable {name=Output.SkeletalMeshLocation003.SampledPositionWithoutOffset, type=NiagaraPosition}}
    )
    Shape Primitive=::NewEnumerator2
    Surface Only=false
    UseOwnerScale=true
   )
   &AddVelocity(
    Velocity=&MergeHelper_VectorSwitch(
     0=<0 0 0>
     1=<0 0 0>
     2=<0 0 0>
     3=<0 0 0>
     4=<0 0 0>
     5=<0 0 0>
     6=&RandomRangeVector(Minimum=<0 0 -500>, Maximum=<0 0 -50>)
     Index=$Particles.VisibilityTag
     Num=::NewEnumerator5
    )
    Velocity Speed Scale=&FloatFromCurve(
     FloatCurve=DataInterface() {propertyValues {Curve {keys [<at 0.4858 -0.1 RCIM_Constant>,<at 0.5755 2 RCIM_Constant>]}}}
     CurveIndex=$System.Age
    )
   ) {module=AddVelocity001}
  ]
  particleUpdateScript [
   &ParticleState()
   /SetVariables(Particles.Spin:Quat4f=&MakeQuaternion(
     Quaternion From=0
     Axis=<0 0 1>
     Angle=&Multiply_Float(A=-1000, B=$System.Age)
     From=?
     To=?
     Stabilization Mode:ENiagara_OrientVectorCount=::NewEnumerator0
     Side Direction:Vector3f=?
     Up Direction:Vector3f=?
     Angle Type=::NewEnumerator0
     Normalize:NiagaraBool=true
    )) {isSetParameters=true}
   /SetVariables(Particles.PositionDust:NiagaraPosition=$Particles.Initial.Position) {isSetParameters=true}
   &SkeletalMeshLocation(
    Bone / Socket Index=1
    Bone Sampling Mode=::NewEnumerator4
    MaskSpawnGroup=false
    Orientation Sampling=::NewEnumerator1
    Position Sampling=::NewEnumerator2
    Sampled Position Offset=&MergeHelper_VectorSwitch(
     0=&SelectVectorFromArray(
      Direct Array Index=&SpawnGroup()
      Vector Selection Array=DataInterface() {propertyValues {FloatData [<0 0 39>, <0 0 39.5>, <0 0 40>]}}
     )
     1=&SelectVectorFromArray(
      Direct Array Index=&SpawnGroup()
      Vector Selection Array=DataInterface() {propertyValues {FloatData [<0 0 39>, <0 0 39.5>, <0 0 40>]}}
     )
     2=&SelectVectorFromArray(
      Direct Array Index=&SpawnGroup()
      Vector Selection Array=DataInterface() {propertyValues {FloatData [<0 0 39>, <0 0 39.5>, <0 0 40>]}}
     )
     3=<0 0 0>
     Index=$Particles.VisibilityTag
     Num=::NewEnumerator2
    )
    Skeletal Mesh=$System.Skeletal Mesh
   )
   /SetVariables(
    Particles.MeshOrientation_Smear:Quat4f=&MultiplyQuaternion(
     Quaternion A=&MultiplyQuaternion(Quaternion A=$Particles.SkeletalMeshLocation.SampledRotation, Quaternion B=$Particles.Initial.MeshOrientation)
     Quaternion B=$Particles.Spin
    )
    Particles.PositionSmear:NiagaraPosition=$Particles.SkeletalMeshLocation.SampledPosition
   ) {isSetParameters=true}
   &SkeletalMeshLocation(
    Bone Sampling Mode=::NewEnumerator4
    MaskSpawnGroup=false
    Orientation Sampling=::NewEnumerator1
    Position Sampling=::NewEnumerator2
    Sampled Position Offset=&MergeHelper_VectorSwitch(0=<0 0 0>, 1=<0 0 0>, 2=<0 0 0>, 3=<0 0 40>, Index=$Particles.VisibilityTag, Num=::NewEnumerator2)
    Skeletal Mesh=$System.Skeletal Mesh
   ) {module=SkeletalMeshLocation001}
   /SetVariables(
    Particles.MeshOrientation_Wind:Quat4f=&MultiplyQuaternion(
     Quaternion A=&MultiplyQuaternion(Quaternion A=$Particles.SkeletalMeshLocation001.SampledRotation, Quaternion B=$Particles.Initial.MeshOrientation)
     Quaternion B=$Particles.Spin
    )
    Particles.Position_Wind:NiagaraPosition=$Particles.SkeletalMeshLocation001.SampledPosition
   ) {isSetParameters=true}
   &SkeletalMeshLocation(
    Bone Sampling Mode=::NewEnumerator4
    MaskSpawnGroup=false
    Orientation Sampling=::NewEnumerator1
    Position Sampling=::NewEnumerator2
    Skeletal Mesh=$System.Skeletal Mesh
   ) {module=SkeletalMeshLocation002}
   /SetVariables(
    Particles.DynamicMaterialParameterSmearBall:Vector4f=&MakeVector4(X=0, Y=0, Z=&RandomRangeFloat(Minimum=0.75, Maximum=1), W=0)
    Particles.MeshOrientationBall:Quat4f=&MultiplyQuaternion(Quaternion A=$Particles.SkeletalMeshLocation002.SampledRotation, Quaternion B=$Particles.Initial.MeshOrientation)
    Particles.Position_Ball_Smear:NiagaraPosition=$Particles.SkeletalMeshLocation002.SampledPosition
   ) {isSetParameters=true}
   &ScaleColor(
    Scale Alpha=&Multiply_Float(
     A=&MergeHelper_FloatSwitch(
      0=&FloatFromCurve(FloatCurve=DataInterface() {propertyValues {Curve {keys [
           {arriveTangent=3.0083, interpMode=RCIM_Cubic, leaveTangent=3.0083, tangentMode=RCTM_User, time=0, value=0}
           {arriveTangent=-0.1558, interpMode=RCIM_Cubic, leaveTangent=-0.1558, tangentMode=RCTM_User, time=0.2, value=1}
           {arriveTangent=-4.2712, interpMode=RCIM_Cubic, tangentMode=RCTM_Break, time=1, value=0}
          ]}}})
      1=&FloatFromCurve(FloatCurve=DataInterface() {propertyValues {Curve {keys [
           {arriveTangent=3.0083, interpMode=RCIM_Cubic, leaveTangent=3.0083, tangentMode=RCTM_User, time=0, value=0}
           {arriveTangent=-0.1558, interpMode=RCIM_Cubic, leaveTangent=-0.1558, tangentMode=RCTM_User, time=0.2, value=1}
           {arriveTangent=-4.2712, interpMode=RCIM_Cubic, tangentMode=RCTM_Break, time=1, value=0}
          ]}}})
      2=&FloatFromCurve(
       FloatCurve=DataInterface() {propertyValues {Curve {keys [
           {arriveTangent=3.0083, interpMode=RCIM_Cubic, leaveTangent=3.0083, tangentMode=RCTM_User, time=0, value=0}
           {arriveTangent=-0.1558, interpMode=RCIM_Cubic, leaveTangent=-0.1558, tangentMode=RCTM_User, time=0.2, value=1}
           {arriveTangent=-4.2712, interpMode=RCIM_Cubic, tangentMode=RCTM_Break, time=1, value=0}
          ]}}}
       Scale Curve=&RandomRangeFloat(Minimum=0.2, Maximum=0.3)
      )
      3=&FloatFromCurve(
       FloatCurve=DataInterface() {propertyValues {Curve {keys [
           {arriveTangent=3.0083, interpMode=RCIM_Cubic, leaveTangent=3.0083, tangentMode=RCTM_User, time=0, value=0}
           {arriveTangent=-0.1558, interpMode=RCIM_Cubic, leaveTangent=-0.1558, tangentMode=RCTM_User, time=0.2, value=1}
           {arriveTangent=-4.2712, interpMode=RCIM_Cubic, tangentMode=RCTM_Break, time=1, value=0}
          ]}}}
       Scale Curve=&RandomRangeFloat(Minimum=0.2, Maximum=0.3)
      )
      4=1
      5=&FloatFromCurve(
       FloatCurve=DataInterface() {propertyValues {Curve {
          defaultValue=0
          keys [
           {interpMode=RCIM_Cubic, tangentMode=RCTM_User, time=0, value=0.2}
           {arriveTangentWeight=0.0567, interpMode=RCIM_Cubic, leaveTangent=-2.4875, tangentMode=RCTM_Break, time=0.17, value=1}
           <at 1 0>
          ]
         }}}
       Scale Curve=&FN_FrontEndFloatModifier(Gameplay=1, Front End=0.4)
      )
      6=&FN_FrontEndFloatModifier(Gameplay=8, Front End=3)
      Index=$Particles.VisibilityTag
      Num=::NewEnumerator5
     )
     B=&FNTimeOfDayScalar(Night Scale=1, Morning Scale=1, Evening Scale=1)
    )
    Scale RGB=&MergeHelper_VectorSwitch(
     0=<1 1 1>
     1=<1 1 1>
     2=<1 1 1>
     3=<1 1 1>
     4=<1 1 1>
     5=<1 1 1>
     6=&FN_FrontEndVector3Modifier(Gameplay=<0.5 0.5 0.5>, Front End=<0.5 0.7 0.5>)
     Index=$Particles.VisibilityTag
     Num=::NewEnumerator5
    )
    ScaleA=true
    ScaleRGB=true
   )
   /SetVariables(Particles.ScaleRandom:NiagaraFloat=&RandomRangeFloat(Minimum=1.15, Maximum=1.25)) {isSetParameters=true}
   &DynamicMaterialParameters(
    Index 0 Param 1=&FloatFromCurve(FloatCurve=DataInterface() {propertyValues {Curve {keys [{interpMode=RCIM_Cubic, tangentMode=RCTM_Break, time=0.5, value=0}, {tangentMode=RCTM_User, time=1, value=1}]}}})
    Module.Param0WriteEnabled=true
    Module.Param1WriteEnabled=true
    Module.Param2WriteEnabled=true
    Module.Param3WriteEnabled=true
   )
   /SetVariables(
    Particles.DynamicMaterialParameter_Wind:Vector4f=&MakeVector4(X=0, Y=1, Z=&RandomRangeFloat(Minimum=0.75, Maximum=1), W=1)
    Particles.Scale_Wind:Vector3f=&Multiply_Vector(A=$Particles.Scale, B=&MakeVector(X=$Particles.ScaleRandom, Y=$Particles.ScaleRandom, Z=0.3))
   ) {isSetParameters=true}
   /SetVariables(Particles.DynamicMaterialParameterDust:Vector4f=&MakeVector4(
     X=&FloatFromCurve(FloatCurve=DataInterface() {propertyValues {Curve {
         defaultValue=1
         keys [
          {arriveTangent=1.7661, interpMode=RCIM_Cubic, leaveTangent=1.7661, tangentMode=RCTM_User, time=0, value=0}
          {arriveTangent=0.0224, arriveTangentWeight=0.3333, interpMode=RCIM_Cubic, leaveTangent=0.0224, tangentMode=RCTM_User, time=1, value=1}
         ]
        }}})
     Y=1
     Z=0
     W=0
    )) {isSetParameters=true}
   /SetVariables(Particles.ScaleDust:Vector3f=&Multiply_Vector(
     A=$Particles.Initial.Scale
     B=&VectorFromCurve(
      VectorCurve=DataInterface() {propertyValues {
        XCurve {
         defaultValue=0
         keys [
          {arriveTangent=2.784, interpMode=RCIM_Cubic, leaveTangent=2.784, tangentMode=RCTM_User, time=0, value=0}
          {arriveTangent=-0.0405, arriveTangentWeight=0.3333, interpMode=RCIM_Cubic, leaveTangent=-0.0405, tangentMode=RCTM_User, time=1, value=1.4287}
         ]
        }
        YCurve {
         defaultValue=0
         keys [
          {arriveTangent=2.784, interpMode=RCIM_Cubic, leaveTangent=2.784, tangentMode=RCTM_User, time=0, value=0}
          {arriveTangent=-0.0405, arriveTangentWeight=0.3333, interpMode=RCIM_Cubic, leaveTangent=-0.0405, tangentMode=RCTM_User, time=1, value=1.4287}
         ]
        }
        ZCurve {
         defaultValue=0
         keys [{
           arriveTangent=0.9995
           arriveTangentWeight=0.1
           interpMode=RCIM_Cubic
           leaveTangent=0.9995
           leaveTangentWeight=0.2333
           tangentMode=RCTM_User
           time=0.3
           value=1
          }]
        }
       }}
      Scale Curve=<2 2 1>
     )
    )) {isSetParameters=true}
   &Drag(
    Drag=&MergeHelper_FloatSwitch(0=0, 1=0, 2=0, 3=0, 4=0, 5=0, 6=5, Index=$Particles.VisibilityTag, Num=::NewEnumerator5)
    Use Linear Drag=true
    Use Rotational Drag=false
   )
   &SolveForcesAndVelocity(Clamp Velocity=false, Limit Acceleration=false)
   /SetVariables(Particles.SizeMult:NiagaraFloat=&FloatFromCurve(FloatCurve=DataInterface() {propertyValues {Curve {keys [<at 0.5589 0.5>,<at 0.5755 1.25>]}}}, CurveIndex=$System.Age)) {isSetParameters=true}
   &ScaleSpriteSize(Scale Factor=&MakeVector2D(X=1, Y=$Particles.SizeMult), Scale Sprite Size Mode=::NewEnumerator2)
   /SetVariables(
    Particles.Linear Color_Outline:LinearColor=&MakeLinearColorFromVectorAndFloat(
     "Vector (RGB)"=<0 0 0>
     "Float (Alpha)"=&FloatFromCurve(
      FloatCurve=DataInterface() {propertyValues {Curve {keys [
          {arriveTangentWeight=0.3333, interpMode=RCIM_Cubic, leaveTangentWeight=0.3333, tangentMode=RCTM_User, time=0, value=1}
          {tangentMode=RCTM_User, time=1, value=0}
         ]}}}
      Scale Curve=3
     )
    )
    Particles.SpriteSize_Outline:Vector2f=&Multiply_Vector2D(A=$Particles.SpriteSize, B=<1.02 1.2>)
   ) {isSetParameters=true}
   &KillParticlesInVolume(
    Kill Shape=::NewEnumerator2
    Kill Volume Enabled:NiagaraBool=&SetBoolByFloatComparison(
     A=&MergeHelper_FloatSwitch(0=0, 1=0, 2=0, 3=0, 4=0, 5=0, 6=1, Index=&SpawnGroup(), Num=::NewEnumerator5)
     B=0
    )
    Origin Offset:Vector3f=<0 0 50>
   ) {module=KillParticlesInVolume001}
  ]
  rendererClasses [&NiagaraMeshRendererProperties, &NiagaraSpriteRendererProperties]
  renderers [
   {rendererClass=&NiagaraMeshRendererProperties, rendererIndex=0}
   {rendererClass=&NiagaraMeshRendererProperties, rendererIndex=1}
   {rendererClass=&NiagaraMeshRendererProperties, rendererIndex=2}
   {rendererClass=&NiagaraMeshRendererProperties, rendererIndex=3}
   {rendererClass=&NiagaraMeshRendererProperties, rendererIndex=4}
   {rendererClass=&NiagaraMeshRendererProperties, rendererIndex=5}
   {rendererClass=&NiagaraSpriteRendererProperties, rendererIndex=6}
   {rendererClass=&NiagaraSpriteRendererProperties, rendererIndex=7}
  ]
 }}
summary {
 systemName=NS_CompanionEmote_SizeTribe_Tail
 emitters [{emitterName=Twirl, rendererClasses [&NiagaraMeshRendererProperties, &NiagaraSpriteRendererProperties]}]
 userVariables [
  User.bIsFrontEnd {type=NiagaraBool}
  User.bIsFrontEndPreview {type=NiagaraBool}
  User.ColorIndex {type=NiagaraInt32}
  User.SpawnRateBall {type=NiagaraFloat, defaultValue=1, description=""}
  User.SpawnRateSpeedLines {type=NiagaraFloat, defaultValue=0, description=""}
  User.SpeedlinesSpawnRate {type=NiagaraFloat, defaultValue=40, description=""}
  User.TailColor {type=LinearColor, description="", defaultValue {struct=LinearColor, value=<rgba 0.5149 0.1529 0.0561 1>}}
  User.TwirlRotatioNRate {type=NiagaraFloat, defaultValue=-1, description=""}
  User.TwirlSpawnRate {type=NiagaraFloat, defaultValue=12, description=""}
 ]
}
