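# Niagara system for a SizeTribe biped companion emote that shows yellow-orange toon star bursts on skeletal mesh bones or sockets.
# Changes to User.BurstIndex trigger the main burst, while BurstIndexDelay changes trigger drifting sparkles, both gated by system age.
# BurstIndex also selects the attachment bone or socket.
# One emitter serves both through spawn groups and visibility tags, feeding five sprite renderers with separate ring, glow and sparkle sizing.
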NiagaraSystem /CosmeticCompanions/Assets/Biped/SizeTribe/FX/Emote/NS_CompanionEmote_SizeTribe_Toon_StarBurst_LWE
dependencies {
 usedDataInterfaces [
  &NiagaraDataInterfaceCurve
  &NiagaraDataInterfaceSpriteRendererInfo
  &NiagaraDataInterfaceArrayFloat
  &NiagaraDataInterfaceSkeletalMesh
 ]
 usedDynamicInputs [&FNTimeOfDayColor, &SelectFloatFromArray]
 usedModules [
  &SystemState
  &self:SystemSpawnScript.SystemScriptSource.SystemScriptGraph.NiagaraNodeAssignment_0.SetVariables_#1
  &self:SystemSpawnScript.SystemScriptSource.SystemScriptGraph.NiagaraNodeAssignment_1.SetVariables_#2
 ]
}
emitters {StarBurst {
  emitterUpdateScript [
   &EmitterState(Life Cycle Mode=::NewEnumerator1, Loop Behavior=::NewEnumerator1, Loop Duration Mode=::NewEnumerator1, UseLoopDelay=false)
   /Game/Effects/Niagara/Modules/Logic/UpdateValueStateChange(InputIntParam:NiagaraInt32=&MakeIntFromFloat(Float=$User.BurstIndex, Conversion Mode=::NewEnumerator1))
   /Niagara/Modules/Emitter/SpawnByTrigger(
    Trigger:NiagaraBool=&Boolean_AndOperation(Bool1=$Emitter.UpdateValueStateChange.Changed, Bool2=&SetBoolByFloatComparison(A=$System.Age, B=0.1))
    TriggerIntervalSwitch=false
   )
   &UpdateValueStateChange(InputIntParam:NiagaraInt32=&MakeIntFromFloat(Float=$User.BurstIndexDelay, Conversion Mode=::NewEnumerator1)) {module=UpdateValueStateChange001}
   &SpawnByTrigger(
    Trigger:NiagaraBool=&Boolean_AndOperation(Bool1=$Emitter.UpdateValueStateChange001.Changed, Bool2=&SetBoolByFloatComparison(A=$System.Age, B=0.1))
    TriggerIntervalSwitch=false
   ) {enabled=false, module=SpawnByTrigger002}
   &SpawnByTrigger(
    Spawn Count:NiagaraInt32=&RandomRangeInt(Minimum=4, Maximum=6)
    Spawn Group=1
    Trigger:NiagaraBool=&Boolean_AndOperation(Bool1=$Emitter.UpdateValueStateChange001.Changed, Bool2=&SetBoolByFloatComparison(A=$System.Age, B=0.1))
    TriggerIntervalSwitch=false
   ) {module=SpawnByTrigger001}
   &SpawnByTrigger(
    Spawn Count:NiagaraInt32=&RandomRangeInt(Minimum=4, Maximum=6)
    Spawn Group=1
    Trigger:NiagaraBool=&Boolean_AndOperation(Bool1=$Emitter.UpdateValueStateChange001.Changed, Bool2=&SetBoolByFloatComparison(A=$System.Age, B=0.1))
    TriggerIntervalSwitch=false
   ) {module=SpawnByTrigger003}
  ]
  particleSpawnScript [
   /SetVariables(
    Particles.PartitionParticles.Partition:NiagaraInt32=&MakeIntFromFloat(Float=$User.BurstIndex, Conversion Mode=::NewEnumerator1)
    Particles.VisibilityTag=&MergeHelper_IntSwitch(
     0=0
     1=&RandomRangeInt(Minimum=1, Maximum=2)
     2=?
     3=?
     4=?
     5=?
     Index=&SpawnGroup()
     Num=::NewEnumerator0
    )
   ) {isSetParameters=true}
   &InitializeParticle(
    Color=&MergeHelper_ColorSwitch(0=<rgba 4 2.4167 0.2 1>, 1=<rgba 4 2.4167 0.2 1>, 2=?, 3=?, Index=&SpawnGroup(), Num=::NewEnumerator0)
    Lifetime=&MergeHelper_FloatSwitch(
     0=0.25
     1=&RandomRangeFloat(Minimum=0.25, Maximum=0.7)
     2=?
     3=?
     Index=&SpawnGroup()
     Num=::NewEnumerator0
    )
    Sprite Rotation Angle=&MergeHelper_FloatSwitch(
     0=&RandomRangeFloat(Minimum=-360, Maximum=360)
     1=&RandomRangeFloat(Minimum=-360, Maximum=360)
     2=45
     3=?
     Index=$Particles.VisibilityTag
     Num=::NewEnumerator1
    )
    Sprite Rotation Mode=::NewEnumerator0
    Sprite Size Mode=::NewEnumerator3
    Uniform Sprite Size=&MergeHelper_FloatSwitch(
     0=&RandomRangeFloat(Minimum=40, Maximum=56)
     1=&RandomRangeFloat(Minimum=4, Maximum=12)
     2=?
     3=?
     Index=&SpawnGroup()
     Num=::NewEnumerator0
    )
    UsePositionOffset=false
    Write Material Random=true
   )
   &SkeletalMeshLocation(
    Bone / Socket Index=$Particles.PartitionParticles.Partition
    Bone Sampling Mode=::NewEnumerator4
    MaskSpawnGroup=false
    Orientation Sampling=::NewEnumerator1
    Sampled Position Offset=&MergeHelper_VectorSwitch(0=<0 0 0>, 1=<0 0 2>, 2=<0 0 2>, 3=?, Index=$Particles.VisibilityTag, Num=::NewEnumerator1)
    Sampled Position Offset Coordinate Space:ENiagaraSkelMeshTransforms=::NewEnumerator0
    Skeletal Mesh=$System.Skeletal Mesh Reader
   ) {module=SkeletalMeshLocation001}
   &ShapeLocation(
    MaskSpawnGroup=false
    Override Seed=false
    Sphere Radius=&MergeHelper_FloatSwitch(
     0=0
     1=&RandomRangeFloat(Minimum=15, Maximum=20)
     2=?
     3=?
     Index=&SpawnGroup()
     Num=::NewEnumerator0
    )
    UseOwnerScale=true
   )
   &AddVelocity(
    Module.Use Falloff Distance=false
    Origin Offset:NiagaraPosition=$Particles.SkeletalMeshLocation001.SampledPosition
    Override Seed=false
    Velocity Mode=::NewEnumerator1
    Velocity Speed=&MergeHelper_FloatSwitch(
     0=0
     1=&RandomRangeFloat(Minimum=10, Maximum=50)
     2=?
     3=?
     Index=&SpawnGroup()
     Num=::NewEnumerator0
    )
   )
   &CameraOffset(Camera Offset Amount=50)
  ]
  particleUpdateScript [
   &ParticleState()
   &ScaleColor(
    Scale Alpha=&MergeHelper_FloatSwitch(
     0=&FloatFromCurve(FloatCurve=DataInterface() {propertyValues {Curve {keys [<at -0.0046 1.0113 RCIM_Cubic>,<at 1.0011 0.002 RCIM_Cubic>]}}}, Scale Curve=0.8)
     1=&FloatFromCurve(
      FloatCurve=DataInterface() {propertyValues {Curve {defaultValue=1, keys [<at 0 0>,<at 0.2 1>,<at 1 0>]}}}
      Scale Curve=&MergeHelper_FloatSwitch(
       0=&RandomRangeFloat(Minimum=1, Maximum=2)
       1=&RandomRangeFloat(Minimum=0.5, Maximum=1.5)
       2=?
       3=?
       Index=$Particles.VisibilityTag
       Num=::NewEnumerator0
      )
     )
     2=?
     3=?
     Index=&SpawnGroup()
     Num=::NewEnumerator0
    )
    ScaleA=true
    ScaleRGB=false
   )
   &Drag(
    Drag=&MergeHelper_FloatSwitch(
     0=0
     1=&RandomRangeFloat(Minimum=2, Maximum=4)
     2=?
     3=?
     Index=&SpawnGroup()
     Num=::NewEnumerator0
    )
    Use Linear Drag=true
    Use Rotational Drag=false
   )
   &SolveForcesAndVelocity(Clamp Velocity=false, Limit Acceleration=false)
   /SetVariables(
    Particles.ColorRing:LinearColor=&MultiplyLinearColor(Linear Color=$Particles.Color, Scale Factor=<1 1 1 1>)
    Particles.SpriteSizeRing:Vector2f=&Multiply_Vector2D_ByFloat(Float=0.6, Vector2D=$Particles.SpriteSize)
   ) {isSetParameters=true}
   /SetVariables(
    Particles.ColorGlow:LinearColor=&MultiplyLinearColor(Linear Color=$Particles.Color, Scale Factor=<1 1 1 0.1>)
    Particles.SpriteSizeGlow:Vector2f=&Multiply_Vector2D_ByFloat(
     Float=&FloatFromCurve(
      FloatCurve=DataInterface() {propertyValues {Curve {
         defaultValue=1
         keys [
          {arriveTangent=17.9167, interpMode=RCIM_Cubic, leaveTangent=17.9167, tangentMode=RCTM_User, time=0, value=0}
          {arriveTangent=-0.3994, interpMode=RCIM_Cubic, leaveTangent=-0.3994, tangentMode=RCTM_User, time=0.1408, value=0.9847}
          {arriveTangent=-2.4114, interpMode=RCIM_Cubic, leaveTangent=-2.4114, tangentMode=RCTM_User, time=1, value=0}
         ]
        }}}
      Scale Curve=1.4
     )
     Vector2D=$Particles.SpriteSize
    )
   ) {isSetParameters=true}
   &DynamicMaterialParameters(
    Index 0 Param 1=1
    Index 0 Param 2=&FloatFromCurve(FloatCurve=DataInterface() {propertyValues {Curve {
        defaultValue=1
        keys [{interpMode=RCIM_Cubic, tangentMode=RCTM_User, time=0, value=0}, {interpMode=RCIM_Cubic, tangentMode=RCTM_User, time=1, value=1}]
       }}})
    Module.Param0WriteEnabled=true
    Module.Param1WriteEnabled=true
    Module.Param2WriteEnabled=true
    Module.Param3WriteEnabled=true
   )
   &SubUVAnimation(
    "Sprite Renderer "=DataInterface() {propertyValues {SpriteRenderer=&self:StarBurst_0.NiagaraSpriteRendererProperties_0}}
    UseEndFrame=false
    UseStartFrame=false
   )
   /SetVariables(Particles.SparklesSizeUpdate:Vector2f=&Multiply_Vector2D_ByFloat(
     Float=&FloatFromCurve(
      FloatCurve=DataInterface() {propertyValues {Curve {defaultValue=0, keys [<at 0 0>,<at 0.105 1>,<at 1 0>]}}}
      Scale Curve=&MergeHelper_FloatSwitch(0=0, 1=1.5, 2=1, 3=?, Index=$Particles.VisibilityTag, Num=::NewEnumerator1)
     )
     Vector2D=$Particles.SpriteSize
    )) {isSetParameters=true}
   &SpriteRotationRate(Rotation Rate=&MergeHelper_FloatSwitch(
     0=0
     1=&RandomRangeFloat(Minimum=-50, Maximum=50)
     2=?
     3=?
     Index=&SpawnGroup()
     Num=::NewEnumerator0
    ))
  ]
  rendererClasses [&NiagaraSpriteRendererProperties]
  renderers [
   {rendererClass=&NiagaraSpriteRendererProperties}
   {rendererClass=&NiagaraSpriteRendererProperties, rendererIndex=1}
   {rendererClass=&NiagaraSpriteRendererProperties, rendererIndex=2}
   {rendererClass=&NiagaraSpriteRendererProperties, rendererIndex=3}
   {rendererClass=&NiagaraSpriteRendererProperties, rendererIndex=4}
  ]
 }}
summary {
 systemName=NS_CompanionEmote_SizeTribe_Toon_StarBurst_LWE
 emitters [{emitterName=StarBurst, rendererClasses [&NiagaraSpriteRendererProperties]}]
 userVariables [
  User.BurstIndex {type=NiagaraFloat, defaultValue=0, description=""}
  User.BurstIndexDelay {type=NiagaraFloat, defaultValue=1, description=""}
  User.FirstTime {type=NiagaraBool}
  User.ImpactLocation {type=NiagaraPosition, description="", defaultValue {}}
  User.NormalizedNotifyProgress {type=NiagaraFloat, defaultValue=0, description=""}
  User.Returning {type=NiagaraBool}
 ]
}
